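# Parent skeletal mesh material for the eyes of the F_MED_BoomShot_Blam player character body, built entirely on material attributes.
# The shared FNEye MF_eyes function produces the base eye attributes, which feed MF_BoomShot_Transition from the character's Toon skin folder.
# The result then passes through MF_TechArtOverrides.
# It supports mesh deformers and takes its shading model from the material expression, with Default Lit as the one in use.
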Material /BRCosmetics/Characters/Player/Female/Medium/Bodies/F_MED_BoomShot_Blam/Materials/M_BoomShot_Eyes_Parent
graph {
 properties {
  bUseMaterialAttributes=true
  bUsedWithMeshDeformer=true
  bUsedWithSkeletalMesh=true
  pixelDepthOffsetMode=PDOM_Legacy
  shadingModel=MSM_FromMaterialExpression
  usedShadingModels=Default Lit
 }
 nodes [MaterialFunctionCall(In=MaterialFunctionCall!(In=MaterialFunctionCall() {materialFunction=/Game/Characters/Shared/FNEye/Materials/MF_eyes}.Result) {materialFunction=/BRCosmetics/Characters/Player/Female/Medium/Bodies/F_MED_BoomShot_Blam/SKINS/Toon/Materials/MF_BoomShot_Transition}.Result) {materialFunction=&MF_TechArtOverrides}]
}
