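# Control Rig for the BoomShot Blam female medium body cosmetic, built to drive corrective deform bones on the character skeleton.
# The graph is a single shared DeformRig function run from Begin Execution, configured through per-region toggles, curve flags, offset transforms and debug switches.
# The hierarchy carries deform helpers at major joints, pecs and glutes, plus dyn_ bones for hair, clothing and grenades, which this graph has no separate nodes for.
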ControlRigBlueprint /BRCosmetics/Characters/Player/Female/Medium/Bodies/F_MED_BoomShot_Blam/Meshes/F_MED_BoomShot_Blam_CtrlRig
graphs [
 RigVMFunctionLibrary {links [], nodes []}
 RigVMModel {nodes [
   RigUnit_BeginExecution = &RigUnit_BeginExecution()
   FN_DeformRig_FNC = RigVMFunctionReferenceNode(
    DeformRig_ Curve="(deformRig_l_upperarm_up_alpha_72_#1=False,deformRig_r_upperarm_up_alpha_73_#2=False,deformRig_l_upperarm_fr_alpha_74_#3=False,deformRig_r_upperarm_fr_alpha_75_CB51B8 ...(truncated)"
    DeformRig_Bool="(upperarm_1_#4=false,elbow_3_#5=false,wrist_7_#6=false,groin_8_#7=false,glute_17_#8=True,knee_9_40C408324C ...(truncated)"
    DeformRig_Debug=(upperarm_debug_18_#4=False,wrist_debug_20_#6=false,lat_debug_26_#9=False,groinGlute_debug_27_#7=False)
    DeformRig_Value="(upperarm_up_18_#10=(Rotation=(X=0.0,Y=-0.0,Z=0.0,W=1.0),Translation=(X=0.0,Y=0.0,Z=8.0),Scale3D=(X=0.70,Y=1.0,Z=1.0)),upperarm_fwd_fr_19_#11=(Rotation=(X=0.0,Y=-0.0,Z=0.0,W=1.0),Translation=(X=0.0, ...(truncated)"
    ExecuteContext=@RigUnit_BeginExecution.ExecutePin
   )
  ]}
]
hierarchy [
 root {loc=<0 0 0>, rot=<0 0 0 1>}
 pelvis {parent=root, loc=<0 -0.5301 98.0571>, rot=<0.0127 0.707 -0.0127 -0.707>}
 spine_01 {parent=pelvis, loc=<6.1208 -0.9328 0>, rot=<0 0 0.034 -0.9994>}
 spine_02 {parent=spine_01, loc=<6.5997 -0.2095 0>, rot=<0 0 0.0238 -0.9997>}
 spine_03 {parent=spine_02, loc=<6.4914 -0.1237 0>, rot=<0 0 0.0149 -0.9999>}
 spine_04 {parent=spine_03, loc=<6.9598 0.0794 0>, rot=<0 0 -0.0132 -0.9999>}
 spine_05 {parent=spine_04, loc=<6.9987 0 0>, rot=<0 0 -0.0015 -1>}
 attach_cape {parent=spine_05, loc=<9.3216 -8.4501 0>, rot=<0.0616 -0.7044 -0.0616 -0.7044>}
 neck_01 {parent=spine_05, loc=<11.3138 0.1529 0>, rot=<0 0 -0.2003 -0.9797>}
 neck_02 {parent=neck_01, loc=<4.5627 0 0>, rot=<0 0 0.0504 -0.9987>}
 head {parent=neck_02, loc=<5.3914 0 0>, rot=<0 0 0.115 -0.9934>}
 dyn_bang_base {parent=head, loc=<8.2336 4.2672 -6.5389>, rot=<0.0029 -0.7071 -0.7071 0.0029>}
 dyn_bang_1 {parent=head, loc=<8.2336 4.2672 -6.5389>, rot=<-0.0029 0.7071 0.7071 -0.0029>}
 dyn_bang_2 {parent=dyn_bang_1, loc=<3.8959 0 0>, rot=<0 0 0 1>}
 dyn_bang_3 {parent=dyn_bang_2, loc=<3.8959 0 0>, rot=<0 0 0 1>}
 dyn_bang_4 {parent=dyn_bang_3, loc=<3.8959 0 0>, rot=<0 0 0 1>}
 dyn_bang_5 {parent=dyn_bang_4, loc=<3.8959 0 0>, rot=<0 0 0 1>}
 dyn_bun_base {parent=head, loc=<10.8808 -7.3798 0.0563>, rot=<0.0029 -0.7071 -0.0029 -0.7071>}
 dyn_bun {parent=dyn_bun_base, loc=<0 0 0>, rot=<0 0 0 1>}
 dyn_l_hair_3_1 {parent=dyn_bun_base, loc=<2.3125 0.8539 0.4476>, rot=<-0.5348 0.2559 -0.1972 -0.7808>}
 dyn_l_hair_3_2 {parent=dyn_l_hair_3_1, loc=<2.7812 -0.04 0>, rot=<0 0 0.313 -0.9498>}
 dyn_l_hair_3_3 {parent=dyn_l_hair_3_2, loc=<2.8982 -0.1214 0>, rot=<0 0 0.2313 -0.9729>}
 dyn_l_hair_01_1 {parent=dyn_bun_base, loc=<1.3285 1.0146 1.4572>, rot=<-0.653 0.2714 -0.2713 -0.6529>}
 dyn_l_hair_01_2 {parent=dyn_l_hair_01_1, loc=<3.2886 0 0>, rot=<0 0 0 1>}
 dyn_l_hair_01_3 {parent=dyn_l_hair_01_2, loc=<3.2886 0 0>, rot=<0.0003 0 0 -1>}
 dyn_l_hair_05_1 {parent=dyn_bun_base, loc=<2.6481 0.3482 0.1576>, rot=<-0.4337 -0.3975 0.2221 -0.7775>}
 dyn_l_hair_05_2 {parent=dyn_l_hair_05_1, loc=<3.5567 0.0052 0>, rot=<0 0 0.4245 -0.9054>}
 dyn_l_hair_05_3 {parent=dyn_l_hair_05_2, loc=<3.379 -0.2291 0>, rot=<0 0 0.1827 -0.9832>}
 dyn_l_hair_04_1 {parent=dyn_bun_base, loc=<2.5949 0.1798 0.453>, rot=<-0.4436 0.1402 0.1128 -0.878>}
 dyn_l_hair_03_2 {parent=dyn_l_hair_04_1, loc=<2.4372 0.1647 0>, rot=<0 0 0.2755 -0.9613>}
 dyn_l_hair_03_3 {parent=dyn_l_hair_03_2, loc=<2.5782 -0.3528 0>, rot=<0 0 0.152 -0.9884>}
 dyn_l_hair_06_1 {parent=dyn_bun_base, loc=<2.8037 0.4121 0.4385>, rot=<-0.5501 -0.2325 0.1182 -0.7933>}
 dyn_l_hair_06_2 {parent=dyn_l_hair_06_1, loc=<2.9534 0 0>, rot=<0 0 0.2444 -0.9697>}
 dyn_l_hair_06_3 {parent=dyn_l_hair_06_2, loc=<2.9534 0 0>, rot=<0 0 0.2444 -0.9697>}
 dyn_l_hair_02_1 {parent=dyn_bun_base, loc=<2.0323 0.4058 0.8901>, rot=<0.8715 -0.1916 0.2701 0.3616>}
 dyn_l_hair_02_2 {parent=dyn_l_hair_02_1, loc=<2.7862 0 -0.0577>, rot=<0 0.2605 0 -0.9655>}
 dyn_l_hair_02_3 {parent=dyn_l_hair_02_2, loc=<2.7437 0 0>, rot=<0 0 0 1>}
 dyn_r_hair_05_1 {parent=dyn_bun_base, loc=<-2.2795 0.7027 0.9299>, rot=<-0.1882 -0.9338 0.2023 -0.2273>}
 dyn_r_hair_05_2 {parent=dyn_r_hair_05_1, loc=<2.9246 0 -0.0358>, rot=<0 0.2016 0 -0.9795>}
 dyn_r_hair_05_3 {parent=dyn_r_hair_05_2, loc=<2.939 0 0>, rot=<0.0007 0.2367 0.0003 -0.9716>}
 dyn_r_hair_04_1 {parent=dyn_bun_base, loc=<-1.9531 1.306 1.591>, rot=<-0.1186 -0.8851 0.4489 -0.0321>}
 dyn_r_hair_04_2 {parent=dyn_r_hair_04_1, loc=<3.5639 0 0.2391>, rot=<0 0.3401 0 -0.9404>}
 dyn_r_hair_04_3 {parent=dyn_r_hair_04_2, loc=<2.7005 0 0>, rot=<-0.0114 0.2051 0.016 -0.9785>}
 dyn_r_hair_06_1 {parent=dyn_bun_base, loc=<-2.4438 0.2849 1.0218>, rot=<-0.1252 -0.835 0.2463 -0.4758>}
 dyn_r_hair_06_2 {parent=dyn_r_hair_06_1, loc=<2.7911 0 -0.0932>, rot=<0 0.2016 0 -0.9795>}
 dyn_r_hair_06_3 {parent=dyn_r_hair_06_2, loc=<2.7936 0 0>, rot=<0 0.0986 0 -0.9951>}
 dyn_r_hair_03_1 {parent=dyn_bun_base, loc=<-1.9196 0.4058 0.8903>, rot=<0.123 -0.8734 0.4511 0.1364>}
 dyn_r_hair_03_2 {parent=dyn_r_hair_03_1, loc=<2.8521 0 -0.3114>, rot=<0 0.3381 0 -0.9411>}
 dyn_r_hair_03_3 {parent=dyn_r_hair_03_2, loc=<2.5502 0 -0.1668>, rot=<0 0.0559 0 -0.9984>}
 dyn_r_hair_01_1 {parent=dyn_bun_base, loc=<-1.1782 1.2878 1.4021>, rot=<0.2284 -0.5987 0.6617 0.3893>}
 dyn_r_hair_01_2 {parent=dyn_r_hair_01_1, loc=<3.1266 0 0.0511>, rot=<0 0.2016 0 -0.9795>}
 dyn_r_hair_01_3 {parent=dyn_r_hair_01_2, loc=<3.1588 0 0>, rot=<0 0.1976 0 -0.9803>}
 dyn_r_hair_02_1 {parent=dyn_bun_base, loc=<-1.2269 1.3534 1.623>, rot=<0.1657 -0.6647 0.7102 0.1625>}
 dyn_r_hair_02_2 {parent=dyn_r_hair_02_1, loc=<3.2022 0 0.1164>, rot=<0 0.3489 0 -0.9372>}
 dyn_r_hair_02_3 {parent=dyn_r_hair_02_2, loc=<3.2587 0 0>, rot=<0 0 0 1>}
 dyn_front_hair {parent=head, loc=<10.6764 9.7029 -0.3308>, rot=<-0.1907 -0.7922 -0.3178 -0.4848>}
 dyn_forehead_hair_1 {parent=head, loc=<10.1969 8.9943 -2.5055>, rot=<-0.1048 -0.8898 -0.2756 -0.3484>}
 dyn_forehead_hair_2 {parent=dyn_forehead_hair_1, loc=<3.5568 0 0>, rot=<0 0 0 1>}
 dyn_necklace {parent=neck_01, loc=<0.8937 5.0426 -0.0428>, rot=<-0.2649 -0.6556 0.2649 -0.6556>}
 attach_backpack {parent=spine_05, loc=<0 -8.35 0>, rot=<-0.0093 -0.707 0.0093 -0.707>}
 clavicle_l {parent=spine_05, loc=<5.4831 -0.1053 -3.2946>, rot=<0.0465 -0.7289 -0.009 -0.6829>}
 upperarm_l {parent=clavicle_l, loc=<9.8658 -1.2136 -0.4105>, rot=<0.0164 -0.3245 -0.0269 -0.9454>}
 lowerarm_l {parent=upperarm_l, loc=<26.465 -0.8345 0.259>, rot=<0.0036 0.0161 -0.2298 -0.9731>}
 hand_l {parent=lowerarm_l, loc=<22.7355 -0.3203 -0.6105>, rot=<-0.6784 -0.0223 0.0735 -0.7307>}
 hand_attach_l {parent=hand_l, loc=<8.5 -3 0>, rot=<0.7071 0.7071 0 0>}
 weapon_l {parent=hand_l, loc=<-0.1974 -3.4111 0.3811>, rot=<0 0 0 1>}
 thumb_01_l {parent=hand_l, loc=<1.3118 -0.3191 -2.459>, rot=<0.6546 -0.4042 0.0664 -0.6354>}
 thumb_02_l {parent=thumb_01_l, loc=<5.9336 -0.0448 0.0401>, rot=<-0.0037 0.0213 0.298 -0.9543>}
 thumb_03_l {parent=thumb_02_l, loc=<2.2223 0.3682 -0.0907>, rot=<0.0118 -0.013 0.0066 -0.9998>}
 index_metacarpal_l {parent=hand_l, loc=<1.6535 0.0183 -2.96>, rot=<0.0158 -0.0124 -0.0173 -0.9996>}
 index_01_l {parent=index_metacarpal_l, loc=<8.3008 0.0357 -0.0524>, rot=<-0.0044 0.0485 0.0924 -0.9945>}
 index_02_l {parent=index_01_l, loc=<3.3587 -0.3004 0.0452>, rot=<-0.002 0.0027 0.3225 -0.9466>}
 index_03_l {parent=index_02_l, loc=<2.5578 -0.123 -0.0252>, rot=<-0.002 0.0067 0.2558 -0.9667>}
 middle_metacarpal_l {parent=hand_l, loc=<2.1856 -0.1725 -0.5984>, rot=<0.006 0.0168 -0.0329 -0.9993>}
 middle_01_l {parent=middle_metacarpal_l, loc=<7.3275 -0.2447 0.0129>, rot=<-0.0208 0.0042 0.0858 -0.9961>}
 middle_02_l {parent=middle_01_l, loc=<4.0123 0.1401 0.1175>, rot=<-0.0218 0.0258 0.3512 -0.9357>}
 middle_03_l {parent=middle_02_l, loc=<2.6077 0.0534 0.1373>, rot=<0.0082 0.0188 0.1708 -0.9851>}
 ring_metacarpal_l {parent=hand_l, loc=<2.2074 -0.3625 1.3638>, rot=<-0.0284 0.0722 -0.0458 -0.9959>}
 ring_01_l {parent=ring_metacarpal_l, loc=<7.0295 -0.6197 -0.1446>, rot=<-0.0489 -0.0252 0.1892 -0.9804>}
 ring_02_l {parent=ring_01_l, loc=<3.6174 0.4647 -0.1027>, rot=<0 0 0.2806 -0.9598>}
 ring_03_l {parent=ring_02_l, loc=<2.4152 0.114 0.0199>, rot=<0 0 0.1189 -0.9929>}
 pinky_metacarpal_l {parent=hand_l, loc=<1.7616 -0.9753 3.3814>, rot=<-0.0397 0.0814 -0.069 -0.9935>}
 pinky_01_l {parent=pinky_metacarpal_l, loc=<7.1709 -0.4831 -0.1094>, rot=<-0.0654 -0.0035 0.2242 -0.9723>}
 pinky_02_l {parent=pinky_01_l, loc=<2.8274 0.0893 -0.1043>, rot=<0.0037 0.0011 0.2854 -0.9584>}
 pinky_03_l {parent=pinky_02_l, loc=<2.2046 0.0307 -0.0078>, rot=<-0.0002 0.0098 0.1445 -0.9895>}
 deform_wrist_dn_base_l {parent=hand_l, loc=<0 -1.5326 0>, rot=<0 0 0 -1>}
 deform_wrist_dn_l {parent=deform_wrist_dn_base_l, loc=<0 0 0>, rot=<0 0 0 1>}
 deform_wrist_up_base_l {parent=hand_l, loc=<0 2.3009 0>, rot=<0 0 0 -1>}
 deform_wrist_up_l {parent=deform_wrist_up_base_l, loc=<0 0 0>, rot=<0 0 0 1>}
 lowerarm_twist_02_l {parent=lowerarm_l, loc=<10.6707 0 0>, rot=<0 0 0 1>}
 lowerarm_twist_01_l {parent=lowerarm_l, loc=<18.316 0 0>, rot=<0 0 0 1>}
 deform_elbow_in_base_l {parent=lowerarm_l, loc=<0.1834 3.6062 -0.2572>, rot=<0 0 0 -1>}
 deform_elbow_in_l {parent=deform_elbow_in_base_l, loc=<0 0 0>, rot=<0 0 0 1>}
 deform_elbow_out_base_l {parent=lowerarm_l, loc=<0 -4 0>, rot=<0 0 0 1>}
 deform_elbow_out_l {parent=deform_elbow_out_base_l, loc=<0 0 0>, rot=<0 0 0 1>}
 upperarm_twist_01_l {parent=upperarm_l, loc=<7.2135 -0.1134 0>, rot=<0 0 0 1>}
 deform_upperarm_bk_base_l {parent=upperarm_twist_01_l, loc=<-6.4053 -4.3746 0.314>, rot=<0 0 0 -1>}
 deform_upperarm_bk_l {parent=deform_upperarm_bk_base_l, loc=<0 0 0>, rot=<0 0 0 1>}
 deform_upperarm_fr_base_l {parent=upperarm_twist_01_l, loc=<-5.6312 5.092 0.3354>, rot=<0 0 0 -1>}
 deform_upperarm_fr_l {parent=deform_upperarm_fr_base_l, loc=<0 0 0>, rot=<0 0 0 1>}
 deform_upperarm_up_base_l {parent=upperarm_twist_01_l, loc=<-8.891 0.4825 3.9517>, rot=<0 0 0 -1>}
 deform_upperarm_up_l {parent=deform_upperarm_up_base_l, loc=<0 0 0>, rot=<0 0 0 1>}
 upperarm_twist_02_l {parent=upperarm_l, loc=<19.7998 -0.1134 0>, rot=<0 0 0 1>}
 clavicle_r {parent=spine_05, loc=<5.4831 -0.1053 3.2946>, rot=<0.7289 0.0465 -0.6829 0.009>}
 upperarm_r {parent=clavicle_r, loc=<-9.8658 1.2136 0.4105>, rot=<0.0164 -0.3245 -0.0269 -0.9454>}
 lowerarm_r {parent=upperarm_r, loc=<-26.465 0.8345 -0.259>, rot=<0.0036 0.0161 -0.2298 -0.9731>}
 hand_r {parent=lowerarm_r, loc=<-22.7355 0.3203 0.6106>, rot=<-0.6784 -0.0223 0.0735 -0.7307>}
 thumb_01_r {parent=hand_r, loc=<-1.3118 0.3191 2.459>, rot=<0.6546 -0.4042 0.0664 -0.6354>}
 thumb_02_r {parent=thumb_01_r, loc=<-5.9336 0.0448 -0.0401>, rot=<-0.0037 0.0213 0.298 -0.9543>}
 thumb_03_r {parent=thumb_02_r, loc=<-2.2223 -0.3682 0.0907>, rot=<0.0118 -0.013 0.0066 -0.9998>}
 index_metacarpal_r {parent=hand_r, loc=<-1.6535 -0.0183 2.96>, rot=<0.0158 -0.0124 -0.0173 -0.9996>}
 index_01_r {parent=index_metacarpal_r, loc=<-8.3008 -0.0357 0.0524>, rot=<-0.0044 0.0485 0.0924 -0.9945>}
 index_02_r {parent=index_01_r, loc=<-3.3587 0.3004 -0.0452>, rot=<-0.002 0.0027 0.3225 -0.9466>}
 index_03_r {parent=index_02_r, loc=<-2.5578 0.123 0.0252>, rot=<-0.002 0.0067 0.2558 -0.9667>}
 middle_metacarpal_r {parent=hand_r, loc=<-2.1856 0.1725 0.5984>, rot=<0.006 0.0168 -0.0329 -0.9993>}
 middle_01_r {parent=middle_metacarpal_r, loc=<-7.3275 0.2447 -0.0129>, rot=<-0.0208 0.0042 0.0858 -0.9961>}
 middle_02_r {parent=middle_01_r, loc=<-4.0123 -0.1401 -0.1175>, rot=<-0.0218 0.0258 0.3512 -0.9357>}
 middle_03_r {parent=middle_02_r, loc=<-2.6077 -0.0534 -0.1373>, rot=<0.0082 0.0188 0.1708 -0.9851>}
 ring_metacarpal_r {parent=hand_r, loc=<-2.2074 0.3625 -1.3638>, rot=<-0.0284 0.0722 -0.0458 -0.9959>}
 ring_01_r {parent=ring_metacarpal_r, loc=<-7.0295 0.6197 0.1446>, rot=<-0.0489 -0.0252 0.1892 -0.9804>}
 ring_02_r {parent=ring_01_r, loc=<-3.6174 -0.4647 0.1027>, rot=<0 0 0.2806 -0.9598>}
 ring_03_r {parent=ring_02_r, loc=<-2.4152 -0.114 -0.0199>, rot=<0 0 0.1189 -0.9929>}
 pinky_metacarpal_r {parent=hand_r, loc=<-1.7616 0.9753 -3.3814>, rot=<-0.0397 0.0814 -0.069 -0.9935>}
 pinky_01_r {parent=pinky_metacarpal_r, loc=<-7.1709 0.4831 0.1094>, rot=<-0.0654 -0.0035 0.2242 -0.9723>}
 pinky_02_r {parent=pinky_01_r, loc=<-2.8274 -0.0893 0.1043>, rot=<0.0037 0.0011 0.2854 -0.9584>}
 pinky_03_r {parent=pinky_02_r, loc=<-2.2046 -0.0307 0.0078>, rot=<-0.0002 0.0098 0.1445 -0.9895>}
 weapon_r {parent=hand_r, loc=<0.1974 3.4111 -0.3811>, rot=<0 0 0 1>}
 hand_attach_r {parent=hand_r, loc=<-8.5001 2.9999 0>, rot=<0 0 -0.7071 -0.7071>}
 deform_wrist_dn_base_r {parent=hand_r, loc=<-0.0001 1.5328 0>, rot=<0 0 0 -1>}
 deform_wrist_dn_r {parent=deform_wrist_dn_base_r, loc=<0 0 0>, rot=<0 0 0 1>}
 deform_wrist_up_base_r {parent=hand_r, loc=<-0.0002 -2.3007 -0.0001>, rot=<0 0 0 -1>}
 deform_wrist_up_r {parent=deform_wrist_up_base_r, loc=<0 0 0>, rot=<0 0 0 1>}
 lowerarm_twist_02_r {parent=lowerarm_r, loc=<-10.6707 0 0>, rot=<0 0 0 -1>}
 lowerarm_twist_01_r {parent=lowerarm_r, loc=<-18.316 0 0>, rot=<0 0 0 -1>}
 deform_elbow_in_base_r {parent=lowerarm_r, loc=<-0.1834 -3.6062 0.2572>, rot=<0 0 0 -1>}
 deform_elbow_in_r {parent=deform_elbow_in_base_r, loc=<0 0 0>, rot=<0 0 0 1>}
 deform_elbow_out_base_r {parent=lowerarm_r, loc=<0 4 -0.0001>, rot=<0 0 0 -1>}
 deform_elbow_out_r {parent=deform_elbow_out_base_r, loc=<0 0 0>, rot=<0 0 0 1>}
 upperarm_twist_01_r {parent=upperarm_r, loc=<-7.2135 0.1134 0>, rot=<0 0 0 -1>}
 deform_upperarm_bk_base_r {parent=upperarm_twist_01_r, loc=<6.4053 4.3746 -0.314>, rot=<0 0 0 -1>}
 deform_upperarm_bk_r {parent=deform_upperarm_bk_base_r, loc=<0 0 0>, rot=<0 0 0 1>}
 deform_upperarm_fr_base_r {parent=upperarm_twist_01_r, loc=<5.6312 -5.092 -0.3354>, rot=<0 0 0 -1>}
 deform_upperarm_fr_r {parent=deform_upperarm_fr_base_r, loc=<0 0 0>, rot=<0 0 0 1>}
 deform_upperarm_up_base_r {parent=upperarm_twist_01_r, loc=<8.891 -0.4825 -3.9517>, rot=<0 0 0 -1>}
 deform_upperarm_up_r {parent=deform_upperarm_up_base_r, loc=<0 0 0>, rot=<0 0 0 1>}
 upperarm_twist_02_r {parent=upperarm_r, loc=<-19.7998 0.1134 0>, rot=<0 0 0 -1>}
 dyn_c_collar_1 {parent=spine_05, loc=<12.5356 -4.5337 0>, rot=<-0.7065 -0.0284 -0.0284 -0.7065>}
 dyn_c_collar_2 {parent=dyn_c_collar_1, loc=<2.2797 0 0>, rot=<0 0 0 1>}
 dyn_l_collar_1 {parent=spine_05, loc=<9.6107 4.0346 -7.0308>, rot=<-0.575 -0.3756 -0.1217 -0.7166>}
 dyn_l_collar_2 {parent=dyn_l_collar_1, loc=<2.7128 0 0>, rot=<0 0 0 1>}
 dyn_r_collar_1 {parent=spine_05, loc=<9.6106 4.0346 7.0308>, rot=<-0.3756 0.575 0.7166 -0.1217>}
 dyn_r_collar_2 {parent=dyn_r_collar_1, loc=<-2.713 0 -0.0001>, rot=<0 0 0 1>}
 dyn_l_jacket_base {parent=spine_04, loc=<-2.8499 8.5793 -10.1032>, rot=<-0.0229 0.5492 0.8346 -0.0347>}
 dyn_l_jacket_1 {parent=dyn_l_jacket_base, loc=<0 0 0>, rot=<0 0 0 1>}
 dyn_l_jacket_2 {parent=dyn_l_jacket_1, loc=<2.8792 0 0>, rot=<0 0 0 1>}
 dyn_l_jacket_3 {parent=dyn_l_jacket_2, loc=<2.8792 0 0>, rot=<0 0 0 1>}
 dyn_r_jacket_base {parent=spine_04, loc=<-2.8494 8.5793 10.1032>, rot=<-0.5492 -0.0229 -0.0347 -0.8346>}
 dyn_r_jacket_1 {parent=dyn_r_jacket_base, loc=<0 0 0>, rot=<0 0 0 1>}
 dyn_r_jacket_2 {parent=dyn_r_jacket_1, loc=<-2.88 0 0>, rot=<0 0 0 1>}
 dyn_r_jacket_3 {parent=dyn_r_jacket_2, loc=<-2.879 0 0>, rot=<0 0 0 1>}
 dyn_zipper {parent=dyn_r_jacket_3, loc=<-0.1413 0.3196 -3.1075>, rot=<0.506 -0.4939 0.506 -0.4939>}
 dyn_c_jacket_base {parent=spine_04, loc=<1.9705 -8.5152 0>, rot=<-0.0294 0.7065 0.7065 -0.0294>}
 dyn_c_jacket_1 {parent=dyn_c_jacket_base, loc=<0 0 0>, rot=<0 0 0 1>}
 dyn_c_jacket_2 {parent=dyn_c_jacket_1, loc=<2.2131 0 0>, rot=<0 0 0 1>}
 dyn_c_jacket_3 {parent=dyn_c_jacket_2, loc=<2.2131 0 0>, rot=<0 0 0 1>}
 deform_pec_aim_base_l {parent=spine_04, loc=<3.2456 9.3309 -1.2237>, rot=<0.0603 -0.4033 0.3318 -0.8507>}
 deform_pec_aim_l {parent=deform_pec_aim_base_l, loc=<0 0 0>, rot=<0 0 0 1>}
 deform_pec_base_l {parent=deform_pec_aim_l, loc=<9.3665 0 0>, rot=<0 0 0 1>}
 deform_pec_l {parent=deform_pec_base_l, loc=<0 0 0>, rot=<0 0 0 1>}
 deform_pec_aim_base_r {parent=spine_04, loc=<3.2452 9.3309 1.2237>, rot=<-0.4033 -0.0603 0.8507 0.3318>}
 deform_pec_aim_r {parent=deform_pec_aim_base_r, loc=<0 0 0>, rot=<0 0 0 1>}
 deform_pec_base_r {parent=deform_pec_aim_r, loc=<-9.3667 -0.0003 -0.0004>, rot=<0 0 0 1>}
 deform_pec_r {parent=deform_pec_base_r, loc=<0 0 0>, rot=<0 0 0 1>}
 thigh_l {parent=pelvis, loc=<-1.0855 -0.2109 -8.7053>, rot=<0.0009 0.0355 0.0155 -0.9993>}
 calf_l {parent=thigh_l, loc=<-42.5284 -0.0019 -0.7297>, rot=<0.0006 -0.0161 -0.0391 -0.9991>}
 foot_l {parent=calf_l, loc=<-46.7401 -2.5014 -0.592>, rot=<0.0427 -0.0238 0.0353 -0.9982>}
 ball_l {parent=foot_l, loc=<-3.5658 10.8066 -0.1544>, rot=<0.0001 -0.0001 -0.7063 -0.7079>}
 calf_twist_02_l {parent=calf_l, loc=<-19.1546 -1.7943 -0.2638>, rot=<0.0237 0.0095 -0.0035 -0.9997>}
 dyn_l_boot_ring_base {parent=calf_twist_02_l, loc=<-2.645 -2.1095 -7.5915>, rot=<-0.0572 -0.7384 -0.0407 -0.6708>}
 dyn_l_boot_ring {parent=dyn_l_boot_ring_base, loc=<0 0 0>, rot=<0 0 0 1>}
 calf_twist_01_l {parent=calf_l, loc=<-35.8483 -1.9197 -0.5794>, rot=<0.0237 0.0095 -0.0035 -0.9997>}
 deform_knee_in_base_l {parent=calf_l, loc=<0.8031 -5.0231 0.2456>, rot=<0 0 0 -1>}
 deform_knee_in_l {parent=deform_knee_in_base_l, loc=<0 0 0>, rot=<0 0 0 1>}
 deform_knee_out_base_l {parent=calf_l, loc=<0 4 0>, rot=<0 0 0 -1>}
 deform_knee_out_l {parent=deform_knee_out_base_l, loc=<0 0 0>, rot=<0 0 0 1>}
 thigh_twist_01_l {parent=thigh_l, loc=<-27.952 0 0>, rot=<0.0013 0 0 -1>}
 deform_groin_base_l {parent=thigh_l, loc=<0.4801 5.6776 -0.2486>, rot=<0 0 0 -1>}
 deform_groin_l {parent=deform_groin_base_l, loc=<0 0 0>, rot=<0 0 0 1>}
 thigh_r {parent=pelvis, loc=<-1.0855 -0.2109 8.7053>, rot=<0.0355 -0.0009 0.9993 0.0155>}
 calf_r {parent=thigh_r, loc=<42.5284 0.0019 0.7297>, rot=<0.0006 -0.0161 -0.0391 -0.9991>}
 foot_r {parent=calf_r, loc=<46.7401 2.5014 0.592>, rot=<0.0427 -0.0238 0.0353 -0.9982>}
 ball_r {parent=foot_r, loc=<3.5658 -10.8066 0.1544>, rot=<0.0001 -0.0001 -0.7063 -0.7079>}
 calf_twist_02_r {parent=calf_r, loc=<19.1546 1.7943 0.2638>, rot=<0.0237 0.0095 -0.0035 -0.9997>}
 dyn_r_boot_ring_base {parent=calf_twist_02_r, loc=<2.645 2.1095 7.5914>, rot=<-0.0572 -0.7384 -0.0407 -0.6708>}
 dyn_r_boot_ring {parent=dyn_r_boot_ring_base, loc=<0 0 0>, rot=<0 0 0 1>}
 calf_twist_01_r {parent=calf_r, loc=<35.8483 1.9197 0.5794>, rot=<0.0237 0.0095 -0.0035 -0.9997>}
 deform_knee_in_base_r {parent=calf_r, loc=<-0.8031 5.0231 -0.2457>, rot=<0 0 0 -1>}
 deform_knee_in_r {parent=deform_knee_in_base_r, loc=<0 0 0>, rot=<0 0 0 1>}
 deform_knee_out_base_r {parent=calf_r, loc=<0 -4 0>, rot=<0 0 0 -1>}
 deform_knee_out_r {parent=deform_knee_out_base_r, loc=<0 0 0>, rot=<0 0 0 1>}
 thigh_twist_01_r {parent=thigh_r, loc=<27.952 0 0>, rot=<0.0013 0 0 -1>}
 dyn_r_grenade_1 {parent=thigh_r, loc=<2.0853 -5.0953 8.9851>, rot=<0.7012 0.1601 -0.6325 -0.2873>}
 dyn_r_grenade_1_ring {parent=dyn_r_grenade_1, loc=<0.1417 6.4276 -1.3196>, rot=<0.0288 -0.0295 -0.0018 -0.9991>}
 dyn_r_grenade_2 {parent=thigh_r, loc=<9.163 -5.7921 10.4875>, rot=<-0.5241 -0.1481 0.7805 0.3069>}
 dyn_r_grenade_2_ring {parent=dyn_r_grenade_2, loc=<0.3108 6.8982 -1.268>, rot=<0.0105 -0.031 0.0191 -0.9993>}
 deform_groin_base_r {parent=thigh_r, loc=<-0.4801 -5.6776 0.2486>, rot=<0 0 0 -1>}
 deform_groin_r {parent=deform_groin_base_r, loc=<0 0 0>, rot=<0 0 0 1>}
 dyn_l_strap_base {parent=pelvis, loc=<7.679 -1.0316 -13.4938>, rot=<0.212 -0.6937 -0.6677 -0.1677>}
 dyn_l_strap_1 {parent=dyn_l_strap_base, loc=<0 0 0>, rot=<0 0 0 1>}
 dyn_l_strap_2 {parent=dyn_l_strap_1, loc=<2.2838 0 0>, rot=<0 0 0 1>}
 dyn_l_strap_3 {parent=dyn_l_strap_2, loc=<2.284 0 0>, rot=<0 0 0 1>}
 dyn_l_grenade {parent=dyn_l_strap_3, loc=<5.0054 -2.2026 1.2133>, rot=<0.633 -0.5073 -0.4071 0.4198>}
 dyn_l_grenade_ring {parent=dyn_l_grenade, loc=<1.6521 6.3485 -2.2487>, rot=<0.3645 -0.0674 0.017 -0.9286>}
 deform_glute_base_l {parent=pelvis, loc=<-7.3466 -9.0213 -9.1245>, rot=<0.0172 0.031 -0.4477 -0.8935>}
 deform_glute_l {parent=deform_glute_base_l, loc=<0 0 0>, rot=<0 0 0 1>}
 deform_glute_base_r {parent=pelvis, loc=<-7.3464 -9.0213 9.1245>, rot=<0.031 -0.0172 0.8935 -0.4477>}
 deform_glute_r {parent=deform_glute_base_r, loc=<0 0 0>, rot=<0 0 0 1>}
 attach {parent=root, loc=<0 0 -4.3002>, rot=<0 0 -0.2 -0.9798>}
 ik_hand_root {parent=root, loc=<0 0 -4.3002>, rot=<0 0 -0.2 -0.9798>}
 ik_hand_gun {parent=ik_hand_root, loc=<-48.1743 3.3545 104.4381>, rot=<0.6276 0.1309 0.4183 -0.6434>}
 ik_hand_r {parent=ik_hand_gun, loc=<0 0 0>, rot=<0 0 0 1>}
 ik_hand_l {parent=ik_hand_gun, loc=<59.3271 67.6997 34.3524>, rot=<-0.6158 -0.7027 -0.3565 0>}
 ik_foot_root {parent=root, loc=<0 0 -4.3002>, rot=<0 0 -0.2 -0.9798>}
 ik_foot_l {parent=ik_foot_root, loc=<-71.0318 -31.5597 123.8208>, rot=<-0.503 0.0519 0.8143 0.285>}
 ik_foot_r {parent=ik_foot_root, loc=<-42.752 -22.6362 186.5898>, rot=<0.2499 0.4381 0.1739 -0.8458>}
 root_fp {parent=root, loc=<0 0 -4.3002>, rot=<0 0 -0.2 -0.9798>}
 attach_fp {parent=root_fp, loc=<0 0 0>, rot=<0 0 0 1>}
 camera_root {parent=root_fp, loc=<0 0 0>, rot=<0 0 0 1>}
 camera {parent=camera_root, loc=<0 0 0>, rot=<0 0 0 1>}
 attach_emote_01 {parent=root, loc=<0 0 -4.3002>, rot=<0 0 -0.2 -0.9798>}
 attach_emote_02 {parent=root, loc=<0 0 -4.3002>, rot=<0 0 -0.2 -0.9798>}
 attach_emote_03 {parent=root, loc=<0 0 -4.3002>, rot=<0 0 -0.2 -0.9798>}
 warp_point_01 {parent=root, loc=<0 0 -4.3002>, rot=<0 0 -0.2 -0.9798>}
 warp_point_02 {parent=root, loc=<0 0 -4.3002>, rot=<0 0 -0.2 -0.9798>}
 warp_point_03 {parent=root, loc=<0 0 -4.3002>, rot=<0 0 -0.2 -0.9798>}
 MaterialCurveA {type=Curve, loc=<0 0 0>, rot=<0 0 0 1>}
]
