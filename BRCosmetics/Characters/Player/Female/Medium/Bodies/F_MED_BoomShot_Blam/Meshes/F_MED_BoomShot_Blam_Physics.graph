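# Physics asset for the F_MED_BoomShot_Blam female medium body, providing capsule bodies and constraints for the dynamic jacket bones.
# Left, right and center jacket chains each hang three links from a base bone that is locked in rotation to spine_04.
# Side links swing within moderate limits, while center links allow only slight swing on one axis, all with zero twist.
# Larger pelvis and spine capsules form the torso, joined by rotationally free constraints.
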PhysicsAsset /BRCosmetics/Characters/Player/Female/Medium/Bodies/F_MED_BoomShot_Blam/Meshes/F_MED_BoomShot_Blam_Physics
bodies {
 dyn_c_jacket_1 [{shapeType=Capsule, length=7.6651, radius=1.0702, shapeName=dyn_c_jacket_1_capsule, center=<1.9898 0 0>, rotation=<rot 0.0002 0 -90>}]
 dyn_c_jacket_2 [{shapeType=Capsule, length=10.0027, radius=1.0702, shapeName=dyn_c_jacket_2_capsule, center=<1.9898 0 0>, rotation=<rot 0.0002 0 -90>}]
 dyn_c_jacket_3 [{shapeType=Capsule, length=11.3077, radius=1.0702, shapeName=dyn_c_jacket_3_capsule, center=<1.9898 0 0>, rotation=<rot 0.0002 0 -90>}]
 dyn_c_jacket_base [{shapeType=Capsule, length=0.6107, radius=1.0702, shapeName=dyn_c_jacket_base_capsule, rotation=<rot 0.0002 0 -90>}]
 dyn_l_jacket_1 [{shapeType=Capsule, length=6.7443, radius=1.403, shapeName=dyn_l_jacket_1_capsule, center=<2 0 0>, rotation=<rot -14.9999 0 0>}]
 dyn_l_jacket_2 [{shapeType=Capsule, length=6.7193, radius=1.3542, shapeName=dyn_l_jacket_2_capsule, center=<2 0 0>, rotation=<rot -14.9999 0 0>}]
 dyn_l_jacket_3 [{shapeType=Capsule, length=6.7193, radius=1.3542, shapeName=dyn_l_jacket_3_capsule, center=<2 0 0>, rotation=<rot -14.9999 0 0>}]
 dyn_l_jacket_base [{shapeType=Capsule, length=0.1537, radius=1.3542, shapeName=dyn_l_jacket_base_capsule, rotation=<rot -14.9999 0 0>}]
 dyn_r_jacket_1 [{shapeType=Capsule, length=6.7443, radius=1.403, shapeName=dyn_l_jacket_1_capsule, center=<-2 0 0>, rotation=<rot -14.9999 0 -180>}]
 dyn_r_jacket_2 [{shapeType=Capsule, length=6.7193, radius=1.3542, shapeName=dyn_l_jacket_2_capsule, center=<-2 0 0>, rotation=<rot -14.9999 0 -180>}]
 dyn_r_jacket_3 [{shapeType=Capsule, length=6.7193, radius=1.3542, shapeName=dyn_l_jacket_3_capsule, center=<-2 0 0>, rotation=<rot -14.9999 0 -180>}]
 dyn_r_jacket_base [{shapeType=Capsule, length=0.1537, radius=1.3542, shapeName=dyn_l_jacket_base_capsule, rotation=<rot -14.9999 0 -180>}]
 pelvis [{shapeType=Capsule, length=5.3264, radius=12.0456, shapeName=pelvis_capsule}]
 spine_02 [{shapeType=Capsule, length=0.3806, radius=11.1129, shapeName=spine_02_capsule}]
 spine_03 [{shapeType=Capsule, length=0.3806, radius=10.1467, shapeName=spine_03_capsule, center=<0 1.4374 0>}]
 spine_04 [{shapeType=Capsule, length=0.3806, radius=10.0128, shapeName=spine_04_capsule, center=<0 2.4417 0>}]
}
constraints [
 {bone1Name=dyn_l_jacket_1, bone2Name=dyn_l_jacket_base, swing1LimitDegrees=40, swing2LimitDegrees=20, twistLimitDegrees=0}
 {bone1Name=dyn_l_jacket_2, bone2Name=dyn_l_jacket_1, swing1LimitDegrees=40, swing2LimitDegrees=20, twistLimitDegrees=0}
 {bone1Name=dyn_l_jacket_3, bone2Name=dyn_l_jacket_2, swing1LimitDegrees=40, swing2LimitDegrees=20, twistLimitDegrees=0}
 {
  bone1Name=dyn_l_jacket_base
  bone2Name=spine_04
  swing1LimitDegrees=45
  swing1Motion=Locked
  swing2LimitDegrees=45
  swing2Motion=Locked
  twistLimitDegrees=45
  twistMotion=Locked
 }
 {
  bone1Name=spine_03
  bone2Name=spine_02
  swing1LimitDegrees=45
  swing1Motion=Free
  swing2LimitDegrees=45
  swing2Motion=Free
  twistLimitDegrees=45
  twistMotion=Free
 }
 {
  bone1Name=spine_04
  bone2Name=spine_03
  swing1LimitDegrees=45
  swing1Motion=Free
  swing2LimitDegrees=45
  swing2Motion=Free
  twistLimitDegrees=45
  twistMotion=Free
 }
 {
  bone1Name=dyn_r_jacket_base
  bone2Name=spine_04
  swing1LimitDegrees=45
  swing1Motion=Locked
  swing2LimitDegrees=45
  swing2Motion=Locked
  twistLimitDegrees=45
  twistMotion=Locked
 }
 {bone1Name=dyn_r_jacket_1, bone2Name=dyn_r_jacket_base, swing1LimitDegrees=40, swing2LimitDegrees=20, twistLimitDegrees=0}
 {bone1Name=dyn_r_jacket_2, bone2Name=dyn_r_jacket_1, swing1LimitDegrees=40, swing2LimitDegrees=20, twistLimitDegrees=0}
 {bone1Name=dyn_r_jacket_3, bone2Name=dyn_r_jacket_2, swing1LimitDegrees=40, swing2LimitDegrees=20, twistLimitDegrees=0}
 {
  bone1Name=spine_02
  bone2Name=pelvis
  swing1LimitDegrees=45
  swing1Motion=Free
  swing2LimitDegrees=45
  swing2Motion=Free
  twistLimitDegrees=45
  twistMotion=Free
 }
 {
  bone1Name=dyn_c_jacket_base
  bone2Name=spine_04
  swing1LimitDegrees=45
  swing1Motion=Locked
  swing2LimitDegrees=45
  swing2Motion=Locked
  twistLimitDegrees=45
  twistMotion=Locked
 }
 {bone1Name=dyn_c_jacket_1, bone2Name=dyn_c_jacket_base, swing1LimitDegrees=0, swing2LimitDegrees=5, twistLimitDegrees=0}
 {bone1Name=dyn_c_jacket_2, bone2Name=dyn_c_jacket_1, swing1LimitDegrees=0, swing2LimitDegrees=5, twistLimitDegrees=0}
 {bone1Name=dyn_c_jacket_3, bone2Name=dyn_c_jacket_2, swing1LimitDegrees=0, swing2LimitDegrees=5, twistLimitDegrees=0}
]
