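# Post-process style AnimBlueprint for the F_MED_BoomShot_Blam female body cosmetic, adding secondary motion to accessories.
# It copies the parent mesh pose, runs a Control Rig, then chains AnimDynamics, constraints and bone offsets for rings, grenades, necklace, zipper, collar and straps.
# A RigidBody pass follows, then DefaultSlot blends onto the left strap branch, weighted by a front-end idle curve.
# The update event derives gravity and tuning variables from crouch, emote, skydive and front-end states.
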AnimBlueprint /BRCosmetics/Characters/Player/Female/Medium/Bodies/F_MED_BoomShot_Blam/Meshes/F_MED_BoomShot_Blam_AnimBP
target_skeleton=&/F_MED_BoomShot_Blam_Skeleton
graphs [
 AnimGraph {nodes [
   n1 = Copy Pose From Mesh() {bCopyCurves=true, bUseAttachedParent=true}
   n2 = Control Rig(Source=Component To Local(ComponentPose=Local To Component(LocalPose=@n1))) {
    alphaBoolBlend {blendOption=Linear}
    controlRigAssetReference {blueprintRigClass=&/F_MED_BoomShot_Blam_CtrlRig.F_MED_BoomShot_Blam_CtrlRig_C}
   }
   n3 = Anim Dynamics - Bone: dyn_l_boot_ring(ComponentPose=Local To Component(LocalPose=@n2), Alpha=1, ExternalForce="0,0,0") {
    angularDampingOverride=1
    angularSpringConstant=10000
    bAngularSpring=true
    bOverrideAngularDamping=true
    bUseGravityOverride=true
    gravityScale=0.35
    simulationSpace=BoneRelative
    alphaBoolBlend {blendOption=Linear}
    boundBone {boneName=dyn_l_boot_ring}
    componentAppliedLinearAccClamp=<5000 5000 5000>
    componentLinearAccScale=<1 1 1>
    componentLinearVelScale=<1 1 1>
    gravityOverride {z=-980}
    physicsBodyDefinitions [{
      boundBone {boneName=dyn_l_boot_ring}
      boxExtents=<30 30 30>
      constraintSetup {angularLimitsMax=<0 0 0>, angularLimitsMin=<0 -60 0>, angularTarget=<1 0 0>}
      localJointOffset=<4 0 -4>
     }]
   }
   n4 = Anim Dynamics - Bone: dyn_r_boot_ring(ComponentPose=@n3, Alpha=1, ExternalForce="0,0,0") {
    angularDampingOverride=1
    angularSpringConstant=10000
    bAngularSpring=true
    bOverrideAngularDamping=true
    bUseGravityOverride=true
    gravityScale=0.35
    simulationSpace=BoneRelative
    alphaBoolBlend {blendOption=Linear}
    boundBone {boneName=dyn_r_boot_ring}
    componentAppliedLinearAccClamp=<5000 5000 5000>
    componentLinearAccScale=<1 1 1>
    componentLinearVelScale=<1 1 1>
    gravityOverride {z=-980}
    physicsBodyDefinitions [{
      boundBone {boneName=dyn_r_boot_ring}
      boxExtents=<30 30 30>
      constraintSetup {angularLimitsMax=<0 0 0>, angularLimitsMin=<0 -60 0>, angularTarget=<1 0 0>}
      localJointOffset=<-4 0 4>
     }]
   }
   n5 = Anim Dynamics - Bone: dyn_r_grenade_1_ring!(SimulationSpace:AnimPhysSimSpaceType Enum, ComponentPose=@n4, Alpha=1, ExternalForce="0,0,0") -> (Pose) {
    angularDampingOverride=1
    angularSpringConstant=20000
    bAngularSpring=true
    bOverrideAngularDamping=true
    bUseGravityOverride=true
    gravityScale=0.35
    simulationSpace=RootRelative
    alphaBoolBlend {blendOption=Linear}
    boundBone {boneName=dyn_r_grenade_1_ring}
    componentAppliedLinearAccClamp=<5000 5000 5000>
    componentLinearAccScale=<1 1 1>
    componentLinearVelScale=<1 1 1>
    gravityOverride {z=-980}
    physicsBodyDefinitions [{
      boundBone {boneName=dyn_r_grenade_1_ring}
      boxExtents=<20 20 20>
      constraintSetup {angularLimitsMax=<15 15 15>, angularLimitsMin=<-15 -15 -15>, angularTarget=<1 0 0>}
      localJointOffset=<0 0 -4>
     }]
   }
   n6 = Anim Dynamics - Bone: dyn_r_grenade_2_ring!(SimulationSpace:AnimPhysSimSpaceType Enum, ComponentPose=@n5, Alpha=1, ExternalForce="0,0,0") -> (Pose) {
    angularDampingOverride=1
    angularSpringConstant=20000
    bAngularSpring=true
    bOverrideAngularDamping=true
    bUseGravityOverride=true
    gravityScale=0.35
    simulationSpace=RootRelative
    alphaBoolBlend {blendOption=Linear}
    boundBone {boneName=dyn_r_grenade_2_ring}
    componentAppliedLinearAccClamp=<5000 5000 5000>
    componentLinearAccScale=<1 1 1>
    componentLinearVelScale=<1 1 1>
    gravityOverride {z=-980}
    physicsBodyDefinitions [{
      boundBone {boneName=dyn_r_grenade_2_ring}
      boxExtents=<20 20 20>
      constraintSetup {angularLimitsMax=<15 15 15>, angularLimitsMin=<-15 -15 -15>, angularTarget=<1 0 0>}
      localJointOffset=<0 0 -4>
     }]
   }
   n7 = "Constraint - dyn_necklace (PPPP)"!(
    ComponentPose=@n6
    Alpha=1
    ConstraintWeights_0=0.2057
    ConstraintWeights_1:float=0.3988
    ConstraintWeights_2:float=0.1532
    ConstraintWeights_3:float=0.2423
   ) -> (Pose) {
    alphaBoolBlend {blendOption=Linear}
    boneToModify {boneName=dyn_necklace}
    constraintSetup [{targetBone {boneName=spine_05}}, {targetBone {boneName=neck_01}}, {targetBone {boneName=head}}, {targetBone {boneName=neck_02}}]
    constraintWeights [1,1,1,1]
   }
   n8 = "Transform (Modify) Bone - Bone: dyn_necklace"(ComponentPose=@n7, Alpha=1, Translation="0,0,0", Rotation="-15,0,-40", Scale="1,1,1") {rotationMode=BMM_Additive, alphaBoolBlend {blendOption=Linear}, boneToModify {boneName=dyn_necklace}, rotation {pitch=-15, roll=-40}}
   n9 = Anim Dynamics - Bone: dyn_necklace(ComponentPose=@n8, Alpha=1, ExternalForce="0,0,0") {
    angularDampingOverride=1
    angularSpringConstant=20000
    bAngularSpring=true
    bOverrideAngularDamping=true
    bUseGravityOverride=true
    gravityScale=0.35
    simulationSpace=RootRelative
    alphaBoolBlend {blendOption=Linear}
    boundBone {boneName=dyn_necklace}
    componentLinearAccScale=<1 1 1>
    componentLinearVelScale=<1 1 1>
    gravityOverride {z=-980}
    physicsBodyDefinitions [{
      boundBone {boneName=dyn_necklace}
      boxExtents=<20 30 30>
      constraintSetup {angularLimitsMax=<0 30 15>, angularLimitsMin=<-30 -30 -15>, angularTarget=<1 0 0>}
      localJointOffset=<0 0 -4>
     }]
   }
   n10 = Anim Dynamics - Bone: dyn_r_grenade_1!(SimulationSpace:AnimPhysSimSpaceType Enum, ComponentPose=@n9, Alpha=1, ExternalForce="0,0,0") -> (Pose) {
    angularDampingOverride=1
    angularSpringConstant=10000
    bAngularSpring=true
    bOverrideAngularDamping=true
    bUseGravityOverride=true
    gravityScale=0.35
    simulationSpace=RootRelative
    alphaBoolBlend {blendOption=Linear}
    boundBone {boneName=dyn_r_grenade_1}
    componentAppliedLinearAccClamp=<5000 5000 5000>
    componentLinearAccScale=<1 1 1>
    componentLinearVelScale=<1 1 1>
    gravityOverride {z=-980}
    physicsBodyDefinitions [{
      boundBone {boneName=dyn_r_grenade_1}
      boxExtents=<15 15 15>
      constraintSetup {angularLimitsMax=<6 6 6>, angularLimitsMin=<-6 -6 -6>, angularTarget=<1 0 0>}
      localJointOffset=<0 0 -2>
     }]
   }
   n11 = Anim Dynamics - Bone: dyn_r_grenade_2!(SimulationSpace:AnimPhysSimSpaceType Enum, ComponentPose=@n10, Alpha=1, ExternalForce="0,0,0") -> (Pose) {
    angularDampingOverride=1
    angularSpringConstant=10000
    bAngularSpring=true
    bOverrideAngularDamping=true
    bUseGravityOverride=true
    gravityScale=0.35
    simulationSpace=RootRelative
    alphaBoolBlend {blendOption=Linear}
    boundBone {boneName=dyn_r_grenade_2}
    componentAppliedLinearAccClamp=<5000 5000 5000>
    componentLinearAccScale=<1 1 1>
    componentLinearVelScale=<1 1 1>
    gravityOverride {z=-980}
    physicsBodyDefinitions [{
      boundBone {boneName=dyn_r_grenade_2}
      boxExtents=<18 18 18>
      constraintSetup {angularLimitsMax=<6 6 6>, angularLimitsMin=<-6 -6 -6>, angularTarget=<1 0 0>}
      localJointOffset=<0 0 -2>
     }]
   }
   n12 = Anim Dynamics - Bone: dyn_zipper(ComponentPose=@n11, Alpha=1) {
    angularDampingOverride=0.999
    angularSpringConstant=10
    bAngularSpring=true
    bOverrideAngularDamping=true
    bUseGravityOverride=true
    gravityScale=0.35
    simulationSpace=RootRelative
    alphaBoolBlend {blendOption=Linear}
    boundBone {boneName=dyn_zipper}
    componentLinearAccScale=<1 1 1>
    componentLinearVelScale=<1 1 1>
    gravityOverride {z=-980}
    physicsBodyDefinitions [{
      boundBone {boneName=dyn_zipper}
      boxExtents=<11 11 11>
      constraintSetup {angularLimitsMax=<30 2 5>, angularLimitsMin=<-30 0 -5>, angularTarget=<1 0 0>}
      localJointOffset=<0 0 -2>
     }]
   }
   n13 = "Constraint - dyn_l_collar_1 (P)"(ComponentPose=@n12, Alpha=0.4, ConstraintWeights_0=1) {
    alphaBoolBlend {blendOption=Linear}
    boneToModify {boneName=dyn_l_collar_1}
    constraintSetup [{targetBone {boneName=clavicle_l}}]
    constraintWeights [1]
   }
   n14 = "Constraint - dyn_r_collar_1 (P)"(ComponentPose=@n13, Alpha=0.2, ConstraintWeights_0=1) {
    alphaBoolBlend {blendOption=Linear}
    boneToModify {boneName=dyn_r_collar_1}
    constraintSetup [{targetBone {boneName=clavicle_r}}]
    constraintWeights [1]
   }
   n15 = "Constraint - dyn_r_collar_1 (P)"(ComponentPose=@n14, Alpha=0.3, ConstraintWeights_0=1) {
    alphaBoolBlend {blendOption=Linear}
    boneToModify {boneName=dyn_r_collar_1}
    constraintSetup [{targetBone {boneName=upperarm_twist_01_r}}]
    constraintWeights [1]
   }
   n16 = "Constraint - dyn_c_collar_1 (P)"(ComponentPose=@n15, Alpha=0.2, ConstraintWeights_0=1) {
    alphaBoolBlend {blendOption=Linear}
    boneToModify {boneName=dyn_c_collar_1}
    constraintSetup [{targetBone {boneName=head}}]
    constraintWeights [1]
   }
   n17 = Anim Dynamics - Bone: dyn_l_strap_1!(SimulationSpace:AnimPhysSimSpaceType Enum, ComponentPose=@n16, Alpha=1, ExternalForce="0,0,0") -> (Pose) {
    angularDampingOverride=1
    angularSpringConstant=5000
    bAngularSpring=true
    bOverrideAngularDamping=true
    bUseGravityOverride=true
    gravityScale=0.35
    simulationSpace=RootRelative
    alphaBoolBlend {blendOption=Linear}
    boundBone {boneName=dyn_l_strap_1}
    componentAppliedLinearAccClamp=<5000 5000 5000>
    componentLinearAccScale=<1 1 1>
    componentLinearVelScale=<1 1 1>
    gravityOverride {z=-980}
    physicsBodyDefinitions [{
      boundBone {boneName=dyn_l_strap_1}
      boxExtents=<30 30 30>
      constraintSetup {angularLimitsMax=<0 5 0>, angularLimitsMin=<0 -5 -15>, angularTarget=<1 0 0>}
      localJointOffset=<4 0 0>
     }]
   }
   n18 = Anim Dynamics - Bone: dyn_l_strap_2!(SimulationSpace:AnimPhysSimSpaceType Enum, ComponentPose=@n17, Alpha=1, ExternalForce="0,0,0") -> (Pose) {
    angularDampingOverride=1
    angularSpringConstant=5000
    bAngularSpring=true
    bOverrideAngularDamping=true
    bUseGravityOverride=true
    gravityScale=0.35
    simulationSpace=RootRelative
    alphaBoolBlend {blendOption=Linear}
    boundBone {boneName=dyn_l_strap_2}
    componentAppliedLinearAccClamp=<5000 5000 5000>
    componentLinearAccScale=<1 1 1>
    componentLinearVelScale=<1 1 1>
    gravityOverride {z=-980}
    physicsBodyDefinitions [{
      boundBone {boneName=dyn_l_strap_2}
      boxExtents=<30 30 30>
      constraintSetup {angularLimitsMax=<0 5 0>, angularLimitsMin=<0 -5 -15>, angularTarget=<1 0 0>}
      localJointOffset=<4 0 0>
     }]
   }
   n19 = Anim Dynamics - Bone: dyn_l_strap_3!(SimulationSpace:AnimPhysSimSpaceType Enum, ComponentPose=@n18, Alpha=1, ExternalForce="0,0,0") -> (Pose) {
    angularDampingOverride=1
    angularSpringConstant=5000
    bAngularSpring=true
    bOverrideAngularDamping=true
    bUseGravityOverride=true
    gravityScale=0.35
    simulationSpace=RootRelative
    alphaBoolBlend {blendOption=Linear}
    boundBone {boneName=dyn_l_strap_3}
    componentAppliedLinearAccClamp=<5000 5000 5000>
    componentLinearAccScale=<1 1 1>
    componentLinearVelScale=<1 1 1>
    gravityOverride {z=-980}
    physicsBodyDefinitions [{
      boundBone {boneName=dyn_l_strap_3}
      boxExtents=<30 30 30>
      constraintSetup {angularLimitsMax=<0 5 0>, angularLimitsMin=<0 -5 -15>, angularTarget=<1 0 0>}
      localJointOffset=<4 0 0>
     }]
   }
   n20 = Anim Dynamics - Bone: dyn_l_grenade(ComponentPose=@n19, Alpha=1, ExternalForce="0,0,0") {
    angularDampingOverride=1
    angularSpringConstant=10000
    bAngularSpring=true
    bOverrideAngularDamping=true
    bUseGravityOverride=true
    gravityScale=0.35
    simulationSpace=RootRelative
    alphaBoolBlend {blendOption=Linear}
    boundBone {boneName=dyn_l_grenade}
    componentAppliedLinearAccClamp=<5000 5000 5000>
    componentLinearAccScale=<1 1 1>
    componentLinearVelScale=<1 1 1>
    gravityOverride {z=-980}
    physicsBodyDefinitions [{
      boundBone {boneName=dyn_l_grenade}
      boxExtents=<15 15 15>
      constraintSetup {angularLimitsMax=<6 6 6>, angularLimitsMin=<-6 -6 -6>, angularTarget=<1 0 0>}
      localJointOffset=<0 0 -2>
     }]
   }
   n21 = Anim Dynamics - Bone: dyn_l_grenade_ring!(SimulationSpace:AnimPhysSimSpaceType Enum, ComponentPose=@n20, Alpha=1, ExternalForce="0,0,0") -> (Pose) {
    angularDampingOverride=1
    angularSpringConstant=20000
    bAngularSpring=true
    bOverrideAngularDamping=true
    bUseGravityOverride=true
    gravityScale=0.35
    simulationSpace=RootRelative
    alphaBoolBlend {blendOption=Linear}
    boundBone {boneName=dyn_l_grenade_ring}
    componentAppliedLinearAccClamp=<5000 5000 5000>
    componentLinearAccScale=<1 1 1>
    componentLinearVelScale=<1 1 1>
    gravityOverride {z=-980}
    physicsBodyDefinitions [{
      boundBone {boneName=dyn_l_grenade_ring}
      boxExtents=<20 20 20>
      constraintSetup {angularLimitsMax=<0 15 0>, angularLimitsMin=<0 -15 0>, angularTarget=<1 0 0>}
      localJointOffset=<0 0 -4>
     }]
   }
   n22 = "Constraint - dyn_l_grenade (P)"(ComponentPose=@n21, Alpha=0.5, ConstraintWeights_0=1) {
    alphaBoolBlend {blendOption=Linear}
    boneToModify {boneName=dyn_l_grenade}
    constraintSetup [{targetBone {boneName=thigh_l}}]
    constraintWeights [1]
   }
   n23 = "Constraint - dyn_r_grenade_1 (PPPP)"!(
    ComponentPose=@n22
    Alpha=1
    ConstraintWeights_0=0.0977
    ConstraintWeights_1:float=0.898
    ConstraintWeights_2:float=0.0043
    ConstraintWeights_3:float=0.0001
   ) -> (Pose) {
    alphaBoolBlend {blendOption=Linear}
    boneToModify {boneName=dyn_r_grenade_1}
    constraintSetup [{targetBone {boneName=pelvis}}, {targetBone {boneName=thigh_twist_01_r}}, {targetBone {boneName=thigh_r}}, {targetBone {boneName=thigh_l}}]
    constraintWeights [1,1,1,1]
   }
   n24 = "Constraint - dyn_r_grenade_2 (P)"!(ComponentPose=@n23, Alpha=1, ConstraintWeights_0=1, ConstraintWeights_1:float=0) -> (Pose) {
    alphaBoolBlend {blendOption=Linear}
    boneToModify {boneName=dyn_r_grenade_2}
    constraintSetup [{targetBone {boneName=thigh_twist_01_r}}]
    constraintWeights [1,0]
   }
   n25 = "Transform (Modify) Bone - Bone: dyn_r_grenade_1"(ComponentPose=@n24, Alpha=1, Translation="0,0,0", Rotation="20,0,0", Scale="1,1,1") {rotationMode=BMM_Additive, alphaBoolBlend {blendOption=Linear}, boneToModify {boneName=dyn_r_grenade_1}, rotation {pitch=20}}
   n26 = "Transform (Modify) Bone - Bone: dyn_r_grenade_2"(ComponentPose=@n25, Alpha=1, Translation="0,0,0", Rotation="30,0,0", Scale="1,1,1") {rotationMode=BMM_Additive, alphaBoolBlend {blendOption=Linear}, boneToModify {boneName=dyn_r_grenade_2}, rotation {pitch=30}}
   n27 = RigidBody!(ComponentPose=@n26, Alpha, OverrideWorldGravity:Vector="0,0,-980") -> (Pose) {
    bClampLinearTranslationLimitToRefPose=true
    bOverrideWorldGravity=true
    overridePhysicsAsset=&/F_MED_BoomShot_Blam_Physics
    simulationSpace=BaseBoneSpace
    alphaBoolBlend {blendOption=Linear}
    baseBoneRef {boneName=spine_05}
    componentAppliedLinearAccClamp=<3000 3000 3000>
    componentLinearAccScale=<0.4 0.3 0.3>
    componentLinearVelScale=<0.4 0.3 0.3>
    simSpaceSettings {velocityScaleZ=0, worldAlpha=0.4, externalLinearDragV=<0.2 0.2>}
   }
   Output Pose(Result=Layered blend per bone(
     BasePose=Component To Local(ComponentPose=@n27)
     BlendPoses_0=Slot 'DefaultSlot'()
     BlendWeights_0=Get CustomIdleIsPlayingFloat() -> (CustomIdleIsPlayingFloat:float)
    ) {layerSetup [{branchFilters [{boneName=dyn_l_strap_base}]}]}) {layerGroup=None}
  ]}
 EventGraph {nodes [
   Event BlueprintUpdateAnimation()
   > n1 = Sequence!(execute) -> (then_0, then_1, then_2:Exec, then_3:Exec, then_4:Exec, then_5:Exec, then_6:Exec, then_7:Exec)
   > n2 = Gravity Override State Function DEPRECATED(CharacterPartAnimInstance=Self-Reference(), GameTimeInSeconds=GetGameTimeInSeconds(), GravityOverrideParams=Get Gravity Override Params())
   > Set GravityOverride(GravityOverride=@n2)
   Set crouchGrenade(execute=@n1.then_1, crouchGrenade:double=SelectFloat(A=0.5, B=0, bPickA=Get bIsCrouching())) -> (Output_Get:double)
   Set GrenadeAlpha(execute=@n1.then_2, GrenadeAlpha:double=SelectFloat(A=0.5, B=1, bPickA=Get bIsPlayingEmote())) -> (Output_Get:double)
   Set ringTwist(execute=@n1.then_3, ringTwist:double=SelectFloat(A=0, B=1, bPickA=Get bIsPlayingEmote())) -> (Output_Get:double)
   n3 = Get bIsSkyDiving()
   Set collarSkydive(execute=@n1.then_4, collarSkydive:double=SelectFloat(A=0.4, B=0.4, bPickA=@n3)) -> (Output_Get:double)
   > Set collarClav(collarClav:double=SelectFloat(A=0, B=0.3, bPickA=@n3)) -> (Output_Get:double)
   n4 = NOT Boolean(A=Get bIsSkydiveFloating())
   > Set collarBack(collarBack:double=SelectFloat(A=0.7, B=0.2, bPickA=AND Boolean(A=@n4, B=@n3))) -> (Output_Get:double)
   Branch(execute=@n1.then_5, Condition=Get bIsInFrontEnd())
   > Set CustomIdleIsPlayingFloat(CustomIdleIsPlayingFloat:float=GetCurveValue(self=GetOwnerFrontEndAnimInstance(), CurveName=Frontend_Idle_Bypass)) -> (Output_Get:float)
   n5 = Branch(execute=@n1.then_6, Condition=Get bIsInFrontEnd())
   > Set simSpaceChange(simSpaceChange:AnimPhysSimSpaceType Enum=World) -> (Output_Get:AnimPhysSimSpaceType Enum)
   Set simSpaceChange(execute=@n5.else)
   Set frontEndSimChange(execute=@n1.then_7, frontEndSimChange:double=SelectFloat(A=0.5, B=0, bPickA=Get bIsInFrontEnd())) -> (Output_Get:double)
  ]}
]
variables [
 collarBack {type=number}
 collarClav {type=number}
 collarSkydive {type=number}
 crouchGrenade {type=number}
 crouchRotator {type=Rotator}
 CustomIdleIsPlayingFloat {type=number}
 frontEndSimChange {type=number}
 Gravity Override Params {type=GravityOverrideParamsStruct}
 GravityOverride {type=Vector}
 GrenadeAlpha {type=number}
 ringTwist {type=number}
 simSpaceChange {type=AnimPhysSimSpaceType}
]
