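# Physics asset for the BoomShot Blam female medium character's face accessory mesh, covering a dynamic bang strand, head, neck and left collar.
# Five capsule bodies form the bang chain, linked by narrow swing limits and zero twist limits.
# The chain's root, sphere base, head and two neck bodies are joined with angularly locked constraints.
# A box on dyn_l_collar_2 has no constraint, alongside the head and neck capsules.
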PhysicsAsset /BRCosmetics/Characters/Player/Female/Medium/Bodies/F_MED_BoomShot_Blam/Meshes/Parts/F_MED_BoomShot_Blam_FaceAcc_Physics
bodies {
 dyn_bang_1 [{shapeType=Capsule, length=10.0107, radius=1.9103, shapeName=dyn_bang_1_capsule, center=<3 0 0>, rotation=<rot 0.0004 0 0>}]
 dyn_bang_2 [{shapeType=Capsule, length=9.0483, radius=1.9103, shapeName=dyn_bang_2_capsule, center=<3 0 0>, rotation=<rot 0.0004 0 0>}]
 dyn_bang_3 [{shapeType=Capsule, length=7.9086, radius=1.9103, shapeName=dyn_bang_3_capsule, center=<3 0 0>, rotation=<rot 0.0004 0 0>}]
 dyn_bang_4 [{shapeType=Capsule, length=0.2996, radius=2.1672, shapeName=dyn_bang_4_capsule, center=<3 0.4335 0>, rotation=<rot 0.0004 0 0>}]
 dyn_bang_5 [{shapeType=Capsule, length=0.7487, radius=1.2501, shapeName=dyn_bang_5_capsule, center=<2.4709 0.7039 1.2507>, rotation=<rot 0.0004 0 0>}]
 dyn_bang_base [{shapeType=Sphere, radius=0.8197, shapeName=dyn_bang_base_sphere}]
 dyn_l_collar_2 [{
   shapeType=Box
   extentX=1.2069
   extentY=4.7138
   extentZ=16.0291
   shapeName=dyn_l_collar_2_box
   center=<-0.2857 0.4444 0.9583>
   rotation=<rot -2.5894 -110.3443 -22.1696>
  }]
 head [{
   shapeType=Capsule
   length=12.9807
   radius=4.1528
   shapeName=head_capsule
   center=<-0.1704 2.8335 -0.2073>
   rotation=<rot -4.5602 77.8238 -90.8058>
  }]
 neck_01 [{shapeType=Capsule, length=0.505, radius=4.7391, shapeName=neck_01_capsule, rotation=<rot 0 -90 0>}]
 neck_02 [{shapeType=Capsule, length=0.505, radius=4.9461, shapeName=neck_02_capsule, rotation=<rot 0 -90 0>}]
}
constraints [
 {bone1Name=dyn_bang_2, bone2Name=dyn_bang_1, swing1LimitDegrees=10, swing2LimitDegrees=20, twistLimitDegrees=0}
 {bone1Name=dyn_bang_3, bone2Name=dyn_bang_2, swing1LimitDegrees=10, swing2LimitDegrees=20, twistLimitDegrees=0}
 {bone1Name=dyn_bang_4, bone2Name=dyn_bang_3, swing1LimitDegrees=10, swing2LimitDegrees=20, twistLimitDegrees=0}
 {bone1Name=dyn_bang_5, bone2Name=dyn_bang_4, swing1LimitDegrees=10, swing2LimitDegrees=20, twistLimitDegrees=0}
 {
  bone1Name=dyn_bang_1
  bone2Name=dyn_bang_base
  swing1LimitDegrees=45
  swing1Motion=Locked
  swing2LimitDegrees=45
  swing2Motion=Locked
  twistLimitDegrees=45
  twistMotion=Locked
 }
 {
  bone1Name=dyn_bang_base
  bone2Name=head
  swing1LimitDegrees=45
  swing1Motion=Locked
  swing2LimitDegrees=45
  swing2Motion=Locked
  twistLimitDegrees=45
  twistMotion=Locked
 }
 {
  bone1Name=head
  bone2Name=neck_02
  swing1LimitDegrees=45
  swing1Motion=Locked
  swing2LimitDegrees=45
  swing2Motion=Locked
  twistLimitDegrees=45
  twistMotion=Locked
 }
 {
  bone1Name=neck_02
  bone2Name=neck_01
  swing1LimitDegrees=45
  swing1Motion=Locked
  swing2LimitDegrees=45
  swing2Motion=Locked
  twistLimitDegrees=45
  twistMotion=Locked
 }
]
